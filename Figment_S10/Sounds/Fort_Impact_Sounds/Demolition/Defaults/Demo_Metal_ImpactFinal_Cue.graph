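# Sound cue for the final metal demolition impact, built from distance-crossfaded layers.
# Each mix blends close, medium and far layers via separate attenuations, with a short random delay on medium and a Distant sound class on medium and far.
# The quality switch gives the first two levels randomized waves per layer, while the last two use fewer waves, with wider pitch ranges on single-wave layers.
# Concurrency settings limit simultaneous playback.
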SoundCue /Figment_S10/Sounds/Fort_Impact_Sounds/Demolition/Defaults/Demo_Metal_ImpactFinal_Cue
root=QualityLevel_0
nodes [
 {
  id=Attenuation_10
  children [Modulator_6]
  properties {attenuationSettings=/Figment_S10/Sounds/Attenuation/Fort_Demolition_Close_Crossfade_Att}
 }
 {id=Attenuation_3, children [Modulator_4], properties {attenuationSettings=/Figment_S10/Sounds/Attenuation/Fort_Demolition_Far_Crossfade_Att}}
 {id=Attenuation_4, children [Delay_1], properties {attenuationSettings=/Figment_S10/Sounds/Attenuation/Fort_Demolition_Medium_Crossfade_Att}}
 {
  id=Attenuation_5
  children [Modulator_2]
  properties {attenuationSettings=&Fort_Demolition_Close_Crossfade_Att}
 }
 {id=Attenuation_8, children [Random_0], properties {attenuationSettings=&Fort_Demolition_Far_Crossfade_Att}}
 {id=Attenuation_9, children [Delay_6], properties {attenuationSettings=&Fort_Demolition_Medium_Crossfade_Att}}
 {id=Delay_1, children [Modulator_3], properties {delayMax=0.2, delayMin=0.1}}
 {id=Delay_6, children [Modulator_11], properties {delayMax=0.2, delayMin=0.1}}
 {id=Mixer_1, children [SoundClass_8,SoundClass_7,SoundClass_6], properties {inputVolume [1,1,1]}}
 {id=Mixer_7, children [SoundClass_2,SoundClass_1,SoundClass_0], properties {inputVolume [1,1,1]}}
 {id=Modulator_11, children [Random_18]}
 {id=Modulator_2, children [Random_4]}
 {id=Modulator_3, children [WavePlayer_14], properties {pitchMax=1.07, pitchMin=0.92}}
 {id=Modulator_4, children [WavePlayer_16], properties {pitchMax=1.1, pitchMin=0.9}}
 {id=Modulator_6, children [Random_15]}
 {id=QualityLevel_0, children [Mixer_7,Mixer_7,Mixer_1,Mixer_1]}
 {id=Random_0, children [WavePlayer_0,WavePlayer_1,WavePlayer_2], properties {weights [1,1,1]}}
 {id=Random_15, children [WavePlayer_43, WavePlayer_44, WavePlayer_45], properties {weights [1,1,1]}}
 {id=Random_18, children [WavePlayer_46, WavePlayer_47, WavePlayer_48], properties {weights [1,1,1]}}
 {id=Random_4, children [WavePlayer_12, WavePlayer_13], properties {bRandomizeWithoutReplacement=false, weights [1,1]}}
 {id=SoundClass_0, children [Attenuation_8], properties {soundClassOverride=/Game/Sounds/FortSoundClassesAndModes/Fort_ImpactSounds_Distant}}
 {id=SoundClass_1, children [Attenuation_9], properties {soundClassOverride=&Fort_ImpactSounds_Distant}}
 {
  id=SoundClass_2
  children [Attenuation_10]
  properties {soundClassOverride=/Game/Sounds/FortSoundClassesAndModes/Fort_ImpactSounds_Default_Close}
 }
 {id=SoundClass_6, children [Attenuation_3], properties {soundClassOverride=&Fort_ImpactSounds_Distant}}
 {id=SoundClass_7, children [Attenuation_4], properties {soundClassOverride=&Fort_ImpactSounds_Distant}}
 {
  id=SoundClass_8
  children [Attenuation_5]
  properties {soundClassOverride=&Fort_ImpactSounds_Default_Close}
 }
 {id=WavePlayer_0, properties {soundWaveAssetPtr=/Figment_S10/Sounds/Fort_Impact_Sounds/Demolition/Metal_Distant_Impact_01}}
 {id=WavePlayer_1, properties {soundWaveAssetPtr=/Figment_S10/Sounds/Fort_Impact_Sounds/Demolition/Metal_Distant_Impact_03}}
 {id=WavePlayer_12, properties {soundWaveAssetPtr=/Game/Sounds/Fort_Impact_Sounds/Demolition/Defaults/final_metal_new_01}}
 {id=WavePlayer_13, properties {soundWaveAssetPtr=/Game/Sounds/Fort_Impact_Sounds/Demolition/Defaults/final_metal_new_02}}
 {id=WavePlayer_14, properties {soundWaveAssetPtr=/Figment_S10/Sounds/Fort_Impact_Sounds/Demolition/metal_close_impact_01}}
 {id=WavePlayer_16, properties {soundWaveAssetPtr=&Metal_Distant_Impact_01}}
 {id=WavePlayer_2, properties {soundWaveAssetPtr=/Figment_S10/Sounds/Fort_Impact_Sounds/Demolition/Metal_Distant_Impact_04}}
 {id=WavePlayer_43, properties {soundWaveAssetPtr=&final_metal_new_01}}
 {id=WavePlayer_44, properties {soundWaveAssetPtr=&final_metal_new_02}}
 {id=WavePlayer_45, properties {soundWaveAssetPtr=/Game/Sounds/Fort_Impact_Sounds/Demolition/Defaults/final_metal_new_03}}
 {id=WavePlayer_46, properties {soundWaveAssetPtr=&metal_close_impact_01}}
 {id=WavePlayer_47, properties {soundWaveAssetPtr=/Figment_S10/Sounds/Fort_Impact_Sounds/Demolition/metal_close_impact_02}}
 {id=WavePlayer_48, properties {soundWaveAssetPtr=/Figment_S10/Sounds/Fort_Impact_Sounds/Demolition/metal_close_impact_03}}
]
properties {
 maxDistance=4500
 volumeMultiplier=0.5
 concurrencyOverrides {maxCount=1}
 concurrencySet [/Game/Sounds/Concurrency_Rules/FourMaximum_StopFurthestOldest_VolumeScale]
}
