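# Sound cue for demolition impacts on metal, layering close, medium and distant samples.
# A quality switch selects among four mixes, from randomized wave pools in the first two to single waves per layer in the last two.
# Each mix routes close and medium layers to the close sound class and a quieter far layer to the distant class, using crossfade attenuations per range.
# The first mix adds a distance crossfade on its far layer; concurrency caps instances.
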SoundCue /Figment_S10/Sounds/Fort_Impact_Sounds/Demolition/Defaults/Demo_Metal_Impact_MLTI_Cue
root=QualityLevel_0
nodes [
 {
  id=Attenuation_0
  children [Modulator_1]
  properties {
   attenuationSettings=/Figment_S10/Sounds/Attenuation/Fort_Demolition_Close_Crossfade_Att
   attenuationOverrides {falloffDistance=900, attenuationShapeExtents {x=100}}
  }
 }
 {
  id=Attenuation_1
  children [Modulator_0]
  properties {attenuationSettings=/Figment_S10/Sounds/Attenuation/Fort_Demolition_Medium_Crossfade_Att, attenuationOverrides {falloffDistance=1600}}
 }
 {
  id=Attenuation_10
  children [Modulator_7]
  properties {attenuationSettings=&Fort_Demolition_Medium_Crossfade_Att, attenuationOverrides {falloffDistance=1600}}
 }
 {
  id=Attenuation_11
  children [Modulator_9]
  properties {
   attenuationSettings=/Figment_S10/Sounds/Attenuation/Fort_Demolition_Far_Crossfade_Att
   attenuationOverrides {falloffDistance=3000, attenuationShapeExtents {x=1000}}
  }
 }
 {
  id=Attenuation_2
  children [Modulator_6]
  properties {
   attenuationSettings=&Fort_Demolition_Far_Crossfade_Att
   attenuationOverrides {falloffDistance=3000, attenuationShapeExtents {x=1000}}
  }
 }
 {
  id=Attenuation_3
  children [Modulator_3]
  properties {
   attenuationSettings=&Fort_Demolition_Close_Crossfade_Att
   attenuationOverrides {falloffDistance=900, attenuationShapeExtents {x=100}}
  }
 }
 {
  id=Attenuation_4
  children [Modulator_2]
  properties {attenuationSettings=&Fort_Demolition_Medium_Crossfade_Att, attenuationOverrides {falloffDistance=1600}}
 }
 {
  id=Attenuation_5
  children [Random_4]
  properties {
   attenuationSettings=&Fort_Demolition_Far_Crossfade_Att
   attenuationOverrides {falloffDistance=3000, attenuationShapeExtents {x=1000}}
  }
 }
 {
  id=Attenuation_6
  children [Modulator_5]
  properties {
   attenuationSettings=&Fort_Demolition_Close_Crossfade_Att
   attenuationOverrides {falloffDistance=900, attenuationShapeExtents {x=100}}
  }
 }
 {
  id=Attenuation_7
  children [Modulator_4]
  properties {attenuationSettings=&Fort_Demolition_Medium_Crossfade_Att, attenuationOverrides {falloffDistance=1600}}
 }
 {
  id=Attenuation_8
  children [Random_7]
  properties {
   attenuationSettings=&Fort_Demolition_Far_Crossfade_Att
   attenuationOverrides {falloffDistance=3000, attenuationShapeExtents {x=1000}}
  }
 }
 {
  id=Attenuation_9
  children [Modulator_8]
  properties {
   attenuationSettings=&Fort_Demolition_Close_Crossfade_Att
   attenuationOverrides {falloffDistance=900, attenuationShapeExtents {x=100}}
  }
 }
 {
  id=DistanceCrossFade_1
  children [Attenuation_5, null]
  properties {crossFadeInput [
    {fadeInDistanceEnd=1000, fadeInDistanceStart=500, fadeOutDistanceEnd=10000, fadeOutDistanceStart=2000, volume=1}
    {fadeInDistanceEnd=0, fadeInDistanceStart=0, fadeOutDistanceEnd=0, fadeOutDistanceStart=0, volume=1}
   ]}
 }
 {id=Mixer_0, children [Attenuation_0, Attenuation_1], properties {inputVolume [1,1]}}
 {id=Mixer_1, children [SoundClass_1,SoundClass_0], properties {inputVolume [1,0.7]}}
 {id=Mixer_2, children [Attenuation_3, Attenuation_4], properties {inputVolume [1,1]}}
 {id=Mixer_3, children [SoundClass_3,SoundClass_2], properties {inputVolume [1,0.7]}}
 {id=Mixer_4, children [Attenuation_6, Attenuation_7], properties {inputVolume [1,1]}}
 {id=Mixer_5, children [SoundClass_5,SoundClass_4], properties {inputVolume [1,0.7]}}
 {id=Mixer_6, children [Attenuation_9, Attenuation_10], properties {inputVolume [1,1]}}
 {id=Mixer_7, children [SoundClass_7,SoundClass_6], properties {inputVolume [1,0.7]}}
 {id=Modulator_0, children [WavePlayer_0], properties {pitchMax=1.8, pitchMin=1, volumeMax=1.3, volumeMin=1.3}}
 {id=Modulator_1, children [WavePlayer_6], properties {pitchMax=1.1, pitchMin=0.9}}
 {id=Modulator_2, children [Random_3], properties {pitchMax=1.8, pitchMin=1, volumeMax=1.3, volumeMin=1.3}}
 {id=Modulator_3, children [Random_5]}
 {id=Modulator_4, children [Random_6], properties {pitchMax=1.8, pitchMin=1, volumeMax=1.3, volumeMin=1.3}}
 {id=Modulator_5, children [Random_8]}
 {id=Modulator_6, children [WavePlayer_3]}
 {id=Modulator_7, children [WavePlayer_11], properties {pitchMax=1.8, pitchMin=1, volumeMax=1.3, volumeMin=1.3}}
 {id=Modulator_8, children [WavePlayer_33], properties {pitchMax=1.1, pitchMin=0.9}}
 {id=Modulator_9, children [WavePlayer_12]}
 {id=QualityLevel_0, children [Mixer_3,Mixer_5,Mixer_1,Mixer_7]}
 {id=Random_3, children [WavePlayer_13, WavePlayer_14, WavePlayer_15], properties {weights [1,1,1]}}
 {id=Random_4, children [WavePlayer_16, WavePlayer_17, WavePlayer_18], properties {weights [1,1,1]}}
 {
  id=Random_5
  children [WavePlayer_19, WavePlayer_20, WavePlayer_21, WavePlayer_22, WavePlayer_23, WavePlayer_24, WavePlayer_25]
  properties {weights [1,1,1,1,1,1,1]}
 }
 {id=Random_6, children [WavePlayer_26, WavePlayer_27, WavePlayer_28], properties {weights [1,1,1]}}
 {id=Random_7, children [WavePlayer_29, WavePlayer_30, WavePlayer_31], properties {weights [1,1,1]}}
 {id=Random_8, children [WavePlayer_32, WavePlayer_34, WavePlayer_35, WavePlayer_36, WavePlayer_38], properties {weights [1,1,1,1,1]}}
 {id=SoundClass_0, children [Attenuation_2], properties {soundClassOverride=/Game/Sounds/FortSoundClassesAndModes/Fort_ImpactSounds_Distant}}
 {id=SoundClass_1, children [Mixer_0], properties {soundClassOverride=/Game/Sounds/FortSoundClassesAndModes/Fort_ImpactSounds_Default_Close}}
 {
  id=SoundClass_2
  children [DistanceCrossFade_1]
  properties {soundClassOverride=&Fort_ImpactSounds_Distant}
 }
 {id=SoundClass_3, children [Mixer_2], properties {soundClassOverride=&Fort_ImpactSounds_Default_Close}}
 {id=SoundClass_4, children [Attenuation_8], properties {soundClassOverride=&Fort_ImpactSounds_Distant}}
 {id=SoundClass_5, children [Mixer_4], properties {soundClassOverride=&Fort_ImpactSounds_Default_Close}}
 {id=SoundClass_6, children [Attenuation_11], properties {soundClassOverride=&Fort_ImpactSounds_Distant}}
 {id=SoundClass_7, children [Mixer_6], properties {soundClassOverride=&Fort_ImpactSounds_Default_Close}}
 {id=WavePlayer_0, properties {soundWaveAssetPtr=/Figment_S10/Sounds/Fort_Impact_Sounds/Demolition/Metal_Big_01}}
 {id=WavePlayer_11, properties {soundWaveAssetPtr=&Metal_Big_01}}
 {id=WavePlayer_12, properties {soundWaveAssetPtr=/Figment_S10/Sounds/Fort_Impact_Sounds/Demolition/Metal_Distant_Impact_01}}
 {id=WavePlayer_13, properties {soundWaveAssetPtr=&Metal_Big_01}}
 {id=WavePlayer_14, properties {soundWaveAssetPtr=/Figment_S10/Sounds/Fort_Impact_Sounds/Demolition/Metal_Big_03}}
 {id=WavePlayer_15, properties {soundWaveAssetPtr=/Figment_S10/Sounds/Fort_Impact_Sounds/Demolition/Metal_Big_04}}
 {id=WavePlayer_16, properties {soundWaveAssetPtr=&Metal_Distant_Impact_01}}
 {id=WavePlayer_17, properties {soundWaveAssetPtr=/Figment_S10/Sounds/Fort_Impact_Sounds/Demolition/Metal_Distant_Impact_03}}
 {id=WavePlayer_18, properties {soundWaveAssetPtr=/Figment_S10/Sounds/Fort_Impact_Sounds/Demolition/Metal_Distant_Impact_04}}
 {id=WavePlayer_19, properties {soundWaveAssetPtr=/Game/Sounds/Fort_Impact_Sounds/Demolition/Defaults/default_metal_impact_new_combined_01}}
 {id=WavePlayer_20, properties {soundWaveAssetPtr=/Game/Sounds/Fort_Impact_Sounds/Demolition/Defaults/default_metal_impact_new_combined_02}}
 {id=WavePlayer_21, properties {soundWaveAssetPtr=/Game/Sounds/Fort_Impact_Sounds/Demolition/Defaults/default_metal_impact_new_combined_03}}
 {id=WavePlayer_22, properties {soundWaveAssetPtr=/Game/Sounds/Fort_Impact_Sounds/Demolition/Defaults/default_metal_impact_new_combined_04}}
 {id=WavePlayer_23, properties {soundWaveAssetPtr=/Game/Sounds/Fort_Impact_Sounds/Demolition/Defaults/default_metal_impact_new_combined_05}}
 {id=WavePlayer_24, properties {soundWaveAssetPtr=/Game/Sounds/Fort_Impact_Sounds/Demolition/Defaults/default_metal_impact_new_combined_06}}
 {id=WavePlayer_25, properties {soundWaveAssetPtr=/Game/Sounds/Fort_Impact_Sounds/Demolition/Defaults/default_metal_impact_new_combined_07}}
 {id=WavePlayer_26, properties {soundWaveAssetPtr=&Metal_Big_01}}
 {id=WavePlayer_27, properties {soundWaveAssetPtr=&Metal_Big_03}}
 {id=WavePlayer_28, properties {soundWaveAssetPtr=&Metal_Big_04}}
 {id=WavePlayer_29, properties {soundWaveAssetPtr=&Metal_Distant_Impact_01}}
 {id=WavePlayer_3, properties {soundWaveAssetPtr=&Metal_Distant_Impact_01}}
 {id=WavePlayer_30, properties {soundWaveAssetPtr=&Metal_Distant_Impact_03}}
 {id=WavePlayer_31, properties {soundWaveAssetPtr=&Metal_Distant_Impact_04}}
 {id=WavePlayer_32, properties {soundWaveAssetPtr=&default_metal_impact_new_combined_01}}
 {id=WavePlayer_33, properties {soundWaveAssetPtr=&default_metal_impact_new_combined_01}}
 {id=WavePlayer_34, properties {soundWaveAssetPtr=&default_metal_impact_new_combined_03}}
 {id=WavePlayer_35, properties {soundWaveAssetPtr=&default_metal_impact_new_combined_04}}
 {id=WavePlayer_36, properties {soundWaveAssetPtr=&default_metal_impact_new_combined_05}}
 {id=WavePlayer_38, properties {soundWaveAssetPtr=&default_metal_impact_new_combined_07}}
 {id=WavePlayer_6, properties {soundWaveAssetPtr=&default_metal_impact_new_combined_01}}
]
properties {
 maxDistance=10000
 volumeMultiplier=0.7
 concurrencyOverrides {maxCount=4}
 concurrencySet [/Game/Sounds/Concurrency_Rules/Demo_FourMaximum_StopFurthestOldest_VolumeScale]
}
