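# Sound cue for wood demolition impacts, capped by concurrency rules.
# Each quality branch layers close, medium and far waves through modulators and crossfade attenuations, routing close plus medium to the close impact sound class and far to the distant class.
# The first two quality branches randomize variations, while the last two use single waves per layer to cut variety.
# Only the first branch adds a distance crossfade node on its far layer.
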SoundCue /Figment_S10/Sounds/Fort_Impact_Sounds/Demolition/Defaults/Demo_Wood_Impact_MLTI_Cue
root=QualityLevel_0
nodes [
 {
  id=Attenuation_0
  children [Modulator_2]
  properties {
   attenuationSettings=/Figment_S10/Sounds/Attenuation/Fort_Demolition_Close_Crossfade_Att
   attenuationOverrides {falloffDistance=900, attenuationShapeExtents {x=100}}
  }
 }
 {
  id=Attenuation_1
  children [Modulator_0]
  properties {attenuationSettings=/Figment_S10/Sounds/Attenuation/Fort_Demolition_Medium_Crossfade_Att, attenuationOverrides {falloffDistance=1600}}
 }
 {
  id=Attenuation_10
  children [Modulator_9]
  properties {attenuationSettings=&Fort_Demolition_Medium_Crossfade_Att, attenuationOverrides {falloffDistance=1600}}
 }
 {
  id=Attenuation_11
  children [Modulator_10]
  properties {
   attenuationSettings=/Figment_S10/Sounds/Attenuation/Fort_Demolition_Far_Crossfade_Att
   attenuationOverrides {falloffDistance=3000, attenuationShapeExtents {x=1000}}
  }
 }
 {
  id=Attenuation_2
  children [Modulator_1]
  properties {
   attenuationSettings=&Fort_Demolition_Far_Crossfade_Att
   attenuationOverrides {falloffDistance=3000, attenuationShapeExtents {x=1000}}
  }
 }
 {
  id=Attenuation_3
  children [Modulator_5]
  properties {
   attenuationSettings=&Fort_Demolition_Close_Crossfade_Att
   attenuationOverrides {falloffDistance=900, attenuationShapeExtents {x=100}}
  }
 }
 {
  id=Attenuation_4
  children [Modulator_3]
  properties {attenuationSettings=&Fort_Demolition_Medium_Crossfade_Att, attenuationOverrides {falloffDistance=1600}}
 }
 {
  id=Attenuation_5
  children [Modulator_4]
  properties {
   attenuationSettings=&Fort_Demolition_Far_Crossfade_Att
   attenuationOverrides {falloffDistance=3000, attenuationShapeExtents {x=1000}}
  }
 }
 {
  id=Attenuation_6
  children [Modulator_8]
  properties {
   attenuationSettings=&Fort_Demolition_Close_Crossfade_Att
   attenuationOverrides {falloffDistance=900, attenuationShapeExtents {x=100}}
  }
 }
 {
  id=Attenuation_7
  children [Modulator_6]
  properties {attenuationSettings=&Fort_Demolition_Medium_Crossfade_Att, attenuationOverrides {falloffDistance=1600}}
 }
 {
  id=Attenuation_8
  children [Modulator_7]
  properties {
   attenuationSettings=&Fort_Demolition_Far_Crossfade_Att
   attenuationOverrides {falloffDistance=3000, attenuationShapeExtents {x=1000}}
  }
 }
 {
  id=Attenuation_9
  children [Modulator_11]
  properties {
   attenuationSettings=&Fort_Demolition_Close_Crossfade_Att
   attenuationOverrides {falloffDistance=900, attenuationShapeExtents {x=100}}
  }
 }
 {
  id=DistanceCrossFade_1
  children [Attenuation_5, null]
  properties {crossFadeInput [
    {fadeInDistanceEnd=1000, fadeInDistanceStart=500, fadeOutDistanceEnd=10000, fadeOutDistanceStart=2000, volume=1}
    {fadeInDistanceEnd=0, fadeInDistanceStart=0, fadeOutDistanceEnd=0, fadeOutDistanceStart=0, volume=1}
   ]}
 }
 {id=Mixer_0, children [Attenuation_0, Attenuation_1], properties {inputVolume [1.2,0.7]}}
 {id=Mixer_1, children [SoundClass_1,SoundClass_0], properties {inputVolume [1,1]}}
 {id=Mixer_2, children [Attenuation_3, Attenuation_4], properties {inputVolume [1.2,0.4]}}
 {id=Mixer_3, children [SoundClass_3,SoundClass_2], properties {inputVolume [1,0.8]}}
 {id=Mixer_4, children [Attenuation_6, Attenuation_7], properties {inputVolume [1.2,0.7]}}
 {id=Mixer_5, children [SoundClass_5,SoundClass_4], properties {inputVolume [1,1]}}
 {id=Mixer_6, children [Attenuation_9, Attenuation_10], properties {inputVolume [1.2,0.7]}}
 {id=Mixer_7, children [SoundClass_7,SoundClass_6], properties {inputVolume [1,1]}}
 {id=Modulator_0, children [WavePlayer_3], properties {pitchMax=1.6, pitchMin=1.5}}
 {id=Modulator_1, children [WavePlayer_4], properties {pitchMax=1.5}}
 {id=Modulator_10, children [WavePlayer_15], properties {pitchMax=1.5}}
 {id=Modulator_11, children [WavePlayer_44], properties {pitchMax=1.1, pitchMin=0.9}}
 {id=Modulator_2, children [WavePlayer_8], properties {pitchMax=1.1, pitchMin=0.9}}
 {id=Modulator_3, children [Random_3], properties {pitchMax=1.6, pitchMin=1.5}}
 {id=Modulator_4, children [Random_4], properties {pitchMax=1.5}}
 {id=Modulator_5, children [Random_5]}
 {id=Modulator_6, children [Random_6], properties {pitchMax=1.6, pitchMin=1.5}}
 {id=Modulator_7, children [Random_7], properties {pitchMax=1.5}}
 {id=Modulator_8, children [Random_8]}
 {id=Modulator_9, children [WavePlayer_14], properties {pitchMax=1.6, pitchMin=1.5}}
 {id=QualityLevel_0, children [Mixer_3,Mixer_5,Mixer_1,Mixer_7]}
 {id=Random_3, children [WavePlayer_19, WavePlayer_18, WavePlayer_17, WavePlayer_16], properties {weights [1,1,1,1]}}
 {id=Random_4, children [WavePlayer_20, WavePlayer_21, WavePlayer_22, WavePlayer_23], properties {weights [1,1,1,1]}}
 {
  id=Random_5
  children [WavePlayer_31, WavePlayer_24, WavePlayer_25, WavePlayer_26, WavePlayer_27, WavePlayer_28, WavePlayer_29, WavePlayer_30]
  properties {weights [1,1,1,1,1,1,1,1]}
 }
 {id=Random_6, children [WavePlayer_35, WavePlayer_34, WavePlayer_33, WavePlayer_32], properties {weights [1,1,1,1]}}
 {id=Random_7, children [WavePlayer_36, WavePlayer_37, WavePlayer_38, WavePlayer_39], properties {weights [1,1,1,1]}}
 {
  id=Random_8
  children [WavePlayer_40, WavePlayer_41, WavePlayer_42, WavePlayer_43, WavePlayer_45, WavePlayer_46]
  properties {weights [1,1,1,1,1,1]}
 }
 {id=SoundClass_0, children [Attenuation_2], properties {soundClassOverride=/Game/Sounds/FortSoundClassesAndModes/Fort_ImpactSounds_Distant}}
 {id=SoundClass_1, children [Mixer_0], properties {soundClassOverride=/Game/Sounds/FortSoundClassesAndModes/Fort_ImpactSounds_Default_Close}}
 {
  id=SoundClass_2
  children [DistanceCrossFade_1]
  properties {soundClassOverride=&Fort_ImpactSounds_Distant}
 }
 {id=SoundClass_3, children [Mixer_2], properties {soundClassOverride=&Fort_ImpactSounds_Default_Close}}
 {id=SoundClass_4, children [Attenuation_8], properties {soundClassOverride=&Fort_ImpactSounds_Distant}}
 {id=SoundClass_5, children [Mixer_4], properties {soundClassOverride=&Fort_ImpactSounds_Default_Close}}
 {id=SoundClass_6, children [Attenuation_11], properties {soundClassOverride=&Fort_ImpactSounds_Distant}}
 {id=SoundClass_7, children [Mixer_6], properties {soundClassOverride=&Fort_ImpactSounds_Default_Close}}
 {id=WavePlayer_14, properties {soundWaveAssetPtr=/Figment_S10/Sounds/Fort_Impact_Sounds/Demolition/wood_impact_01}}
 {id=WavePlayer_15, properties {soundWaveAssetPtr=/Figment_S10/Sounds/Fort_Impact_Sounds/Demolition/wood_distant_impact_02}}
 {id=WavePlayer_16, properties {soundWaveAssetPtr=/Figment_S10/Sounds/Fort_Impact_Sounds/Demolition/wood_impact_09}}
 {id=WavePlayer_17, properties {soundWaveAssetPtr=/Figment_S10/Sounds/Fort_Impact_Sounds/Demolition/wood_impact_07}}
 {id=WavePlayer_18, properties {soundWaveAssetPtr=/Figment_S10/Sounds/Fort_Impact_Sounds/Demolition/wood_impact_02}}
 {id=WavePlayer_19, properties {soundWaveAssetPtr=&wood_impact_01}}
 {id=WavePlayer_20, properties {soundWaveAssetPtr=&wood_distant_impact_02}}
 {id=WavePlayer_21, properties {soundWaveAssetPtr=/Figment_S10/Sounds/Fort_Impact_Sounds/Demolition/wood_distant_impact_04}}
 {id=WavePlayer_22, properties {soundWaveAssetPtr=/Figment_S10/Sounds/Fort_Impact_Sounds/Demolition/wood_distant_impact_05}}
 {id=WavePlayer_23, properties {soundWaveAssetPtr=/Figment_S10/Sounds/Fort_Impact_Sounds/Demolition/wood_distant_impact_06}}
 {id=WavePlayer_24, properties {soundWaveAssetPtr=/Game/Sounds/Fort_Impact_Sounds/Demolition/Defaults/default_wood_impact_new_combined_02}}
 {id=WavePlayer_25, properties {soundWaveAssetPtr=/Game/Sounds/Fort_Impact_Sounds/Demolition/Defaults/default_wood_impact_new_combined_03}}
 {id=WavePlayer_26, properties {soundWaveAssetPtr=/Game/Sounds/Fort_Impact_Sounds/Demolition/Defaults/default_wood_impact_new_combined_04}}
 {id=WavePlayer_27, properties {soundWaveAssetPtr=/Game/Sounds/Fort_Impact_Sounds/Demolition/Defaults/default_wood_impact_new_combined_05}}
 {id=WavePlayer_28, properties {soundWaveAssetPtr=/Game/Sounds/Fort_Impact_Sounds/Demolition/Defaults/default_wood_impact_new_combined_06}}
 {id=WavePlayer_29, properties {soundWaveAssetPtr=/Game/Sounds/Fort_Impact_Sounds/Demolition/Defaults/default_wood_impact_new_combined_07}}
 {id=WavePlayer_3, properties {soundWaveAssetPtr=&wood_impact_01}}
 {id=WavePlayer_30, properties {soundWaveAssetPtr=/Game/Sounds/Fort_Impact_Sounds/Demolition/Defaults/default_wood_impact_new_combined_08}}
 {id=WavePlayer_31, properties {soundWaveAssetPtr=/Game/Sounds/Fort_Impact_Sounds/Demolition/Defaults/default_wood_impact_new_combined_01}}
 {id=WavePlayer_32, properties {soundWaveAssetPtr=&wood_impact_09}}
 {id=WavePlayer_33, properties {soundWaveAssetPtr=&wood_impact_07}}
 {id=WavePlayer_34, properties {soundWaveAssetPtr=&wood_impact_02}}
 {id=WavePlayer_35, properties {soundWaveAssetPtr=&wood_impact_01}}
 {id=WavePlayer_36, properties {soundWaveAssetPtr=&wood_distant_impact_02}}
 {id=WavePlayer_37, properties {soundWaveAssetPtr=&wood_distant_impact_04}}
 {id=WavePlayer_38, properties {soundWaveAssetPtr=&wood_distant_impact_05}}
 {id=WavePlayer_39, properties {soundWaveAssetPtr=&wood_distant_impact_06}}
 {id=WavePlayer_4, properties {soundWaveAssetPtr=&wood_distant_impact_02}}
 {id=WavePlayer_40, properties {soundWaveAssetPtr=&default_wood_impact_new_combined_02}}
 {id=WavePlayer_41, properties {soundWaveAssetPtr=&default_wood_impact_new_combined_03}}
 {id=WavePlayer_42, properties {soundWaveAssetPtr=&default_wood_impact_new_combined_04}}
 {id=WavePlayer_43, properties {soundWaveAssetPtr=&default_wood_impact_new_combined_05}}
 {id=WavePlayer_44, properties {soundWaveAssetPtr=&default_wood_impact_new_combined_02}}
 {id=WavePlayer_45, properties {soundWaveAssetPtr=&default_wood_impact_new_combined_07}}
 {id=WavePlayer_46, properties {soundWaveAssetPtr=&default_wood_impact_new_combined_08}}
 {id=WavePlayer_8, properties {soundWaveAssetPtr=&default_wood_impact_new_combined_02}}
]
properties {
 maxDistance=10000
 volumeMultiplier=0.7
 concurrencyOverrides {maxCount=4}
 concurrencySet [/Game/Sounds/Concurrency_Rules/Demo_FourMaximum_StopFurthestOldest_VolumeScale]
}
